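; Gameplay ability for the Season 25 AloeCrouton Q01 audio log, a quest voice-over.
; Activation holds only a disabled parent-call placeholder, so this graph adds no playback logic.
; On end, cancelled or not, it broadcasts CustomVerb.VOFinished with the owning player pawn.
; It then removes gameplay effects by asset tags and by the VO-playing handle, and stops and destroys the looping sound component.
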(event Ability|EventActivateAbility
  (|CallParentFunction) ; This node is disabled and will not be called.\nDrag off pins to build functionality.
)

(event Ability|EventOnEndAbility
  (params (bWasCancelled :type Boolean))
  (bind _loopingsoundhandle (Variables|Data|GetLoopingSoundHandle))
  (CustomVerb|BroadcastCustomVerbMessage (CustomVerb|MakeCustomVerbMessage (Variables|GAVars|Data|GetOwningPlayerPawn) "(TagName=\"CustomVerb.VOFinished\")"))
  (Ability|RemoveGameplayEffectFromOwnerWithAssetTags (Variables|Setup|GetAssetTags))
  (Ability|RemoveGameplayEffectFromOwnerWithHandle (Variables|Data|GetGEVOPlayingHandle) 1)
  (Audio|Components|Audio|Stop _loopingsoundhandle)
  (Components|DestroyComponent _loopingsoundhandle))
